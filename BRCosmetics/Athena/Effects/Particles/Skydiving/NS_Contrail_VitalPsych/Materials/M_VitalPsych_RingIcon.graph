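# Translucent, unlit, two-sided Niagara mesh and ribbon material for the skull and bunny icons of the VitalPsych skydiving contrail.
# Particle random picks a cell in a noise texture, whose rounded value selects between the skull and bunny alphas as a tiled mask.
# Particle color tints that mask, and a hologram function, time-offset by a dynamic parameter, adds its emissive before emissive adjustment.
# A separate branch fades the mask by inverted vertex green.
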Material /BRCosmetics/Athena/Effects/Particles/Skydiving/NS_Contrail_VitalPsych/Materials/M_VitalPsych_RingIcon
graph {
 properties {
  bUsedWithNiagaraMeshParticles=true
  bUsedWithNiagaraRibbons=true
  blendMode=BLEND_Translucent
  pixelDepthOffsetMode=PDOM_Legacy
  shadingModel=MSM_Unlit
  twoSided=true
  usedShadingModels=Unlit
 }
 nodes [
  AppendVector(A=Constant(), B=Constant())
  n1 = ParticleColor()
  n2 = Multiply(
   A=AppendVector(A=ScalarParameter() {defaultValue=1, parameterName="U_Tiling(C)"}, B=ScalarParameter() {defaultValue=1, parameterName="V_Tiling(C)"})
   B=TextureCoordinate()
  )
  n3 = TextureSampleParameter2D(UVs=@n2) {parameterName=TextureInput_B, texture=/BRCosmetics/Athena/Effects/Particles/Skydiving/NS_Contrail_VitalPsych/Textures/T_VitalPsych_Skull}
  n4 = TextureSampleParameter2D(UVs=@n2) {parameterName=TextureInput_A, texture=/BRCosmetics/Athena/Effects/Particles/Skydiving/NS_Contrail_VitalPsych/Textures/T_VitalPsych_Bunny}
  n5 = Constant() {r=32}
  n6 = TextureSample!(
   UVs=MaterialFunctionCall!(UseTexCoord 1, Rows=@n5, Columns=@n5, Particle Random Value=ParticleRandom()) {materialFunction=/Game/Effects/Fort_Effects/Materials/_Functions/MF_RandomSubUV_GPU_Sprites}.Result
   Tex
   Level
   Apply View MipBias
  ) {
   automaticViewMipBias=false
   mipValueMode=TMVM_MipLevel
   samplerType=SAMPLERTYPE_LinearColor
   texture=/Game/Packages/Fortress_Sky/Materials/Noise32x322_nearestNeighboor
  }
  n7 = Multiply(
   A=MaterialFunctionCall!(1=@n3.A, 2=@n3.A, 3=@n4.A, 4, A=Round(None=@n6.G)) {materialFunction=/Engine/Functions/Engine_MaterialFunctions02/Lerp_Multiple_Float}.Lerp 3 Inputs
   B=@n1.A
  )
  n8 = Add(
   A=Time() {bOverride_Period=true, period=6000}
   B=DynamicParameter() {defaultValue=<rgba 0 1 1 1>, paramNames [des,RanOffset,Param3,Param4]}.RanOffset
  )
  n9 = ScalarParameter() {group=Base_HoloFX_Glitch, parameterName=Glitch_UseCustomFlicker}
  n10 = NamedRerouteUsage()
  n11 = NamedRerouteUsage()
  n12 = NamedRerouteUsage()
  MaterialFunctionCall(Emissive=Add(
    A=Multiply(A=@n1.RGB, B=@n7)
    B=MaterialFunctionCall!(
     Time=@n8
     MeshScaleIn
     UVForScanline
     CustomFlicker
     UseCustomFlicker=@n9
     UVForStroke
     UVForGlitch
     BaseUVIn
     EmissiveIn=@n10
     OpacityIn=@n11
     WPOIn=@n12
    ) {materialFunction=/Game/Tools/3DHB/Shader/Base/MF_3DHB_HologramFX}.EmissiveOut
   )) {materialFunction=&MF_FX_EmissiveAdjustments}
  n13 = VertexColor()
  Multiply(A=@n7, B=Power(Base=OneMinus(None=@n13.G)))
 ]
}
