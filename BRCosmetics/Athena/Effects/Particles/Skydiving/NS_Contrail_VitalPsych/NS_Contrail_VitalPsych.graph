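# Niagara skydiving contrail cosmetic "VitalPsych" that dresses the diving character in glows, rings and streaks.
# Eight emitters follow paired skeletal mesh bones or sockets: arm and foot streak sprites, soft glows, eye-logo and hand glow sprites, lights, and spinning mesh rings.
# Dive velocity aligns streaks and drives ring motion and facing, while paraglide lean adjusts offsets and spin.
# Time-of-day scalars tune ring brightness, and front-end modifiers adjust trails for menus.
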NiagaraSystem /BRCosmetics/Athena/Effects/Particles/Skydiving/NS_Contrail_VitalPsych/NS_Contrail_VitalPsych
dependencies {
 usedDataInterfaces [
  &NiagaraDataInterfaceArrayInt32
  &NiagaraDataInterfaceMeshRendererInfo
  &NiagaraDataInterfaceArrayFloat
  &NiagaraDataInterfaceCurve
  &NiagaraDataInterfaceVector2DCurve
  &NiagaraDataInterfaceArrayColor
  &NiagaraDataInterfaceColorCurve
 ]
 usedDynamicInputs [
  &MakeCustomIntFromBool
  &UniformRangedFloat
  &NormalizeVector
  &self:Abs_Z_Axis
  &Multiply_Vector
 ]
 usedModules [
  &SystemState
  &InterpolateOverTime
  &self:SystemSpawnScript.SystemScriptSource.SystemScriptGraph.NiagaraNodeAssignment_2.SetVariables_#1
  &FN_Integrator
  &self:SystemSpawnScript.SystemScriptSource.SystemScriptGraph.NiagaraNodeAssignment_0.SetVariables_#2
  &self:SystemSpawnScript.SystemScriptSource.SystemScriptGraph.NiagaraNodeAssignment_1.SetVariables_#3
 ]
}
emitters {
 Eye_Logo {
  emitterUpdateScript [
   &EmitterState(Life Cycle Mode=::NewEnumerator1, Loop Behavior=::NewEnumerator1, Loop Duration Mode=::NewEnumerator1, UseLoopDelay=false)
   &SpawnBurst_Instantaneous(Spawn Count=2, Use Spawn Probability=false, UseLoopCountLimit=false)
  ]
  particleSpawnScript [
   &PartitionParticles()
   &InitializeParticle(Lifetime=2, Sprite Size Mode=::NewEnumerator3, Uniform Sprite Size=35, UsePositionOffset=false, Write Material Random=true)
  ]
  particleUpdateScript [
   /SetVariables(
    Particles.Color_SoftGlow:LinearColor=&MakeLinearColorFromVectorAndFloat(
     "Vector (RGB)"=&MakeVectorFromLinearColorRGB(Color=<rgba 1.6974 0.2648 2 1>)
     "Float (Alpha)"=&FloatFromCurve(
      FloatCurve=DataInterface() {propertyValues {Curve {keys [<at 0.0001 0.3382>,<at 0.1393 0.6251>,<at 0.3202 0.5233>,<at 0.4613 0.8934>,<at 0.6241 0.3845>,<at 0.7724 0.6066>,<at 1.0006 0.3567>]}}}
      Scale Curve=0.09
     )
    )
    Particles.Size_SoftGlow:Vector2f=<50 50>
   ) {isSetParameters=true}
   &SkeletalMeshLocation(
    Bone / Socket Index=&SelectIntFromArray(Direct Array Index=$Particles.PartitionParticles.Partition, Int Selection Array=DataInterface() {propertyValues {IntData [6,7]}})
    Bone Sampling Mode=::NewEnumerator4
    MaskSpawnGroup=false
    Sampled Position Offset=&Lerp_Vector(A=<0 0 -13>, B=<0 0 -20>, Alpha=$User.ParaGlideLeanAlpha)
    Sampled Position Offset Coordinate Space:ENiagaraSkelMeshTransforms=::NewEnumerator2
    Skeletal Mesh=$User.SkeletalMesh
   )
   &ParticleState(
    Kill Particles When Lifetime Has Elapsed=false
    Let Infinitely Lived Particles Die When Emitter Deactivates=true
    Loop Particles Lifetime=true
   )
   &SolveForcesAndVelocity(Clamp Velocity=false, Limit Acceleration=false)
  ]
  rendererClasses [&NiagaraSpriteRendererProperties]
  renderers [
   {rendererClass=&NiagaraSpriteRendererProperties}
   {rendererClass=&NiagaraSpriteRendererProperties, rendererIndex=1}
  ]
 }
 GlowRings {
  emitterUpdateScript [
   &EmitterState()
   &SpawnRate(SpawnRate=3, Use Spawn Probability=false)
  ]
  particleSpawnScript [
   &PartitionParticles()
   &InitializeParticle(
    Color Mode=::NewEnumerator0
    Mesh Renderer Array Visibility Mode=::NewEnumerator3
    Mesh Renderer Info=DataInterface() {propertyValues {MeshRenderer=&self:GlowRings_0.NiagaraMeshRendererProperties_1}}
    Mesh Scale Mode=::NewEnumerator3
    Sprite Size Mode=::NewEnumerator3
    UsePositionOffset=false
    Write Material Random=true
   )
   &SkeletalMeshLocation(
    Bone / Socket Index=&SelectIntFromArray(Direct Array Index=$Particles.PartitionParticles.Partition, Int Selection Array=DataInterface() {propertyValues {IntData [0,1]}})
    Bone Sampling Mode=::NewEnumerator4
    MaskSpawnGroup=false
    Orientation Sampling=::NewEnumerator1
    Skeletal Mesh=$User.SkeletalMesh
   ) {module=SkeletalMeshLocation001}
   &AddVelocity(
    Velocity=&TransformVector(Vector=$System.DiveVelocityScale, Source Space=::World, Destination Space=::Local)
    Velocity Speed Scale=&RandomRangeFloat(Minimum=-125, Maximum=-150)
   )
   &Color(Color=&MakeLinearColorFromVectorAndFloat("Vector (RGB)"=<1 1 1>, "Float (Alpha)"=&RandomRangeFloat(Minimum=0.7, Maximum=1)))
   &InitialMeshOrientation(
    Mesh Orientation Mode=::NewEnumerator7
    Rotation=&RandomRangeVector(Minimum=<0 0 1>, Maximum=<0 0 -1>)
    Use Rotation Vector=true
   )
  ]
  particleUpdateScript [
   &ParticleState()
   &UpdateMeshOrientation(
    Rotation Rate=&FloatFromCurve(
     FloatCurve=DataInterface() {propertyValues {Curve {keys [
         {interpMode=RCIM_Cubic, tangentMode=RCTM_User, time=0, value=1}
         {arriveTangent=11.4005, interpMode=RCIM_Cubic, leaveTangent=11.4005, tangentMode=RCTM_User, time=1, value=5}
        ]}}}
     Scale Curve=&Lerp_Float(
      A=&RandomRangeFloat(Minimum=0.1, Maximum=0.3)
      B=&RandomRangeFloat(Minimum=0.5, Maximum=0.7)
      Alpha=$User.ParaGlideLeanAlpha
     )
    )
    Rotation Vector=<0 0 1>
   )
   &UpdateMeshOrientation(
    Facing Direction=&TransformVector(Vector=$System.DiveVelocityScale, Source Space=::World, Destination Space=::Local)
    Facing Interpolation=false
    Orientation Method=::NewEnumerator1
    Override Quat=false
    Reference Axis Vector:Vector3f=<0 0 -1>
    Stabilization Mode:ENiagara_OrientVectorCount=::NewEnumerator0
    TransformFacing=false
   ) {module=UpdateMeshOrientation002}
   &SolveForcesAndVelocity(Clamp Velocity=false, Limit Acceleration=false)
   &ScaleMeshSize(Scale Factor=&VectorFromFloat(Value=&FloatFromCurve(
      FloatCurve=DataInterface() {propertyValues {Curve {keys [
          {interpMode=RCIM_Cubic, tangentMode=RCTM_User, time=0, value=0.1}
          {interpMode=RCIM_Cubic, tangentMode=RCTM_Break, time=0.2, value=1}
          {arriveTangent=-2.2488, interpMode=RCIM_Cubic, leaveTangent=-2.2488, tangentMode=RCTM_User, time=1, value=0.5}
         ]}}}
      Scale Curve=1.2
     ))) {module=ScaleMeshSize001}
   &ScaleColor(
    Scale Alpha=&FloatFromCurve(
     FloatCurve=DataInterface() {propertyValues {Curve {keys [
         {interpMode=RCIM_Cubic, tangentMode=RCTM_User, time=0, value=1}
         {arriveTangent=-4.4168, interpMode=RCIM_Cubic, leaveTangent=-4.4168, tangentMode=RCTM_User, time=1, value=0}
        ]}}}
     Scale Curve=&Multiply_Float(A=3, B=&FNTimeOfDayScalar(Night Scale=0.6, Morning Scale=1, Evening Scale=1.8))
    )
    Scale RGB=&MakeVectorFromLinearColorRGB(Color=&ColorFromCurve(ColorCurve=DataInterface() {propertyValues {
        AlphaCurve {keys [<at 0 1 RCIM_Cubic>,<at 1 1 RCIM_Cubic>]}
        BlueCurve {keys [<at 0 2>,<at 0.3416 1.6667>,<at 0.5926 0.484>,<at 1 0.0725 RCIM_Cubic>]}
        GreenCurve {keys [<at 0 0>,<at 0.3416 0>,<at 0.5926 0>,<at 1 0.3722 RCIM_Cubic>]}
        RedCurve {keys [<at 0 0.1759>,<at 0.3416 2>,<at 0.5926 2>,<at 1 2 RCIM_Cubic>]}
       }}))
    ScaleA=true
    ScaleRGB=true
   )
   /SetVariables(
    Particles.Color_Rings_Inner:LinearColor=&ScaleLinearColorByCurve(
     InitialValue=&MakeLinearColorFromVectorAndFloat(
      "Vector (RGB)"=<1 1 1>
      "Float (Alpha)"=&FloatFromCurve(
       FloatCurve=DataInterface() {propertyValues {Curve {keys [
           {interpMode=RCIM_Cubic, tangentMode=RCTM_User, time=0, value=1}
           {arriveTangent=-4.4168, interpMode=RCIM_Cubic, leaveTangent=-4.4168, tangentMode=RCTM_User, time=1, value=0}
          ]}}}
       Scale Curve=&Multiply_Float(A=1.7, B=&FNTimeOfDayScalar(Night Scale=0.8, Morning Scale=0.8, Evening Scale=1))
      )
     )
     ScaleCurve=DataInterface() {propertyValues {
       AlphaCurve {keys [<at 0 1 RCIM_Cubic>,<at 1 1 RCIM_Cubic>]}
       BlueCurve {keys [<at 0 1>,<at 0.3416 1>,<at 0.5926 1>,<at 0.7635 0.2276>,<at 1 0.0706>]}
       GreenCurve {keys [<at 0 0>,<at 0.3416 0>,<at 0.5926 0>,<at 0.7635 0>,<at 1 0.3725>]}
       RedCurve {keys [<at 0 0.1776>,<at 0.3416 0.1156>,<at 0.5926 0.9311>,<at 0.7635 1>,<at 1 1>]}
      }}
     CurveIndex=$Particles.NormalizedAge
    )
    Particles.Size_Rings_Inner:Vector3f=&VectorFromFloat(Value=&FloatFromCurve(
      FloatCurve=DataInterface() {propertyValues {Curve {keys [
          {interpMode=RCIM_Cubic, tangentMode=RCTM_User, time=0, value=0.1}
          {interpMode=RCIM_Cubic, tangentMode=RCTM_Break, time=0.2, value=1}
          {arriveTangent=-2.2488, interpMode=RCIM_Cubic, leaveTangent=-2.2488, tangentMode=RCTM_User, time=1, value=0.5}
         ]}}}
      Scale Curve=1.15
     ))
   ) {isSetParameters=true}
   &DynamicMaterialParameters(
    Index 0 Param 2=&RandomRangeFloat(Minimum=-2000, Maximum=6000)
    Module.Param0WriteEnabled=false
    Module.Param1WriteEnabled=true
    Module.Param2WriteEnabled=false
    Module.Param3WriteEnabled=false
   )
  ]
  rendererClasses [&NiagaraMeshRendererProperties]
  renderers [
   {rendererClass=&NiagaraMeshRendererProperties, rendererIndex=0}
   {rendererClass=&NiagaraMeshRendererProperties, rendererIndex=1}
  ]
 }
 HandGlow_Decal {
  emitterUpdateScript [
   &EmitterState()
   &SpawnBurst_Instantaneous(Spawn Count=2, Use Spawn Probability=false, UseLoopCountLimit=true)
  ]
  particleSpawnScript [
   &InitializeParticle(
    Color=<rgba 0.657 0.0413 1 0.5>
    Lifetime=0.7
    Sprite Size Mode=::NewEnumerator3
    Uniform Sprite Size=50
    UsePositionOffset=false
    Write Material Random=true
   )
   &PartitionParticles()
  ]
  particleUpdateScript [
   &ParticleState(
    Kill Particles When Lifetime Has Elapsed=false
    Let Infinitely Lived Particles Die When Emitter Deactivates=true
    Loop Particles Lifetime=true
   )
   &SkeletalMeshLocation(
    Bone / Socket Index=&SelectIntFromArray(Direct Array Index=$Particles.PartitionParticles.Partition, Int Selection Array=DataInterface() {propertyValues {IntData [6,7]}})
    Bone Sampling Mode=::NewEnumerator4
    MaskSpawnGroup=false
    Skeletal Mesh=$User.SkeletalMesh
   )
   &ScaleColor(
    Scale Alpha=&FloatFromCurve(FloatCurve=DataInterface() {propertyValues {Curve {keys [<at 0.0001 0.3382>,<at 0.1393 0.6251>,<at 0.3202 0.5233>,<at 0.4613 0.8934>,<at 0.6241 0.3845>,<at 0.7724 0.6066>,<at 1.0006 0.3567>]}}})
    ScaleA=true
    ScaleRGB=false
   )
  ]
  rendererClasses [&NiagaraSpriteRendererProperties]
  renderers [{rendererClass=&NiagaraSpriteRendererProperties}]
 }
 Light {
  emitterUpdateScript [
   &EmitterState(Life Cycle Mode=::NewEnumerator1, Loop Delay=1, Loop Duration=1, UseLoopDelay=true)
   &SpawnBurst_Instantaneous(Spawn Count=2, Use Spawn Probability=false, UseLoopCountLimit=false)
   &SpawnBurst_Instantaneous(Spawn Count=2, Spawn Time=2, Use Spawn Probability=false, UseLoopCountLimit=false) {module=SpawnBurst_Instantaneous001}
  ]
  particleSpawnScript [
   &InitializeParticle(
    Color=&SelectLinearColorFromArray(
     Array Sampling Mode=::NewEnumerator0
     Direct Array Index=?
     Color Selection Array=DataInterface() {propertyValues {ColorData [<rgba 0.4698 0.0333 1 22>, <rgba 0 0.7449 1 22>]}}
     Random Seed=0
    )
    Lifetime=2
    Sprite Size Mode=::NewEnumerator3
    Uniform Sprite Size=50
    UsePositionOffset=false
    Write Material Random=true
   )
   &PartitionParticles()
  ]
  particleUpdateScript [
   &ParticleState()
   &SkeletalMeshLocation(
    Bone / Socket Index=&SelectIntFromArray(Direct Array Index=$Particles.PartitionParticles.Partition, Int Selection Array=DataInterface() {propertyValues {IntData [6,7]}})
    Bone Sampling Mode=::NewEnumerator4
    MaskSpawnGroup=false
    Skeletal Mesh=$User.SkeletalMesh
   )
   &ScaleColor(ScaleA=true, ScaleRGB=false) {module=ScaleColor001}
  ]
  rendererClasses [&NiagaraLightRendererProperties]
  renderers [{rendererClass=&NiagaraLightRendererProperties, rendererIndex=0}]
 }
 SoftGlow {
  emitterUpdateScript [
   &EmitterState(Life Cycle Mode=::NewEnumerator1, Loop Behavior=::NewEnumerator1, Loop Duration Mode=::NewEnumerator1, UseLoopDelay=false)
   &SpawnBurst_Instantaneous(Spawn Count=2, Use Spawn Probability=false, UseLoopCountLimit=false)
  ]
  particleSpawnScript [
   &PartitionParticles(Number Of Partitions=4)
   &InitializeParticle(
    Color=<rgba 1.6974 0.2648 2 0.1>
    Lifetime=2
    Sprite Size Mode=::NewEnumerator3
    Uniform Sprite Size=55
    UsePositionOffset=false
    Write Material Random=true
   )
  ]
  particleUpdateScript [
   &SkeletalMeshLocation(
    Bone / Socket Index=&SelectIntFromArray(Direct Array Index=$Particles.PartitionParticles.Partition, Int Selection Array=DataInterface() {propertyValues {IntData [4,5]}})
    Bone Sampling Mode=::NewEnumerator4
    MaskSpawnGroup=false
    Sampled Position Offset Coordinate Space:ENiagaraSkelMeshTransforms=::NewEnumerator2
    Skeletal Mesh=$User.SkeletalMesh
   ) {module=SkeletalMeshLocation001}
   &ParticleState(
    Kill Particles When Lifetime Has Elapsed=false
    Let Infinitely Lived Particles Die When Emitter Deactivates=true
    Loop Particles Lifetime=true
   )
   &SolveForcesAndVelocity(Clamp Velocity=false, Limit Acceleration=false)
   &ScaleColor(
    Scale Alpha=&Multiply_Float(
     A=&Multiply_Float(A=1, B=0.5)
     B=&Multiply_Float(
      A=&Sine(Normalized Angle=$Emitter.Age, Period=0.7, Scale=0.2, Bias=1)
      B=&Sine(Normalized Angle=$Emitter.Age, Period=1, Scale=0.3, Bias=1)
     )
    )
    ScaleA=true
    ScaleRGB=true
   ) {module=ScaleColor001}
   &CameraOffset(Camera Offset Amount=20)
  ]
  rendererClasses [&NiagaraSpriteRendererProperties]
  renderers [{rendererClass=&NiagaraSpriteRendererProperties}]
 }
 Trail_Feet {
  emitterUpdateScript [
   &EmitterState(Life Cycle Mode=::NewEnumerator1, Loop Behavior=::NewEnumerator1, Loop Duration Mode=::NewEnumerator1, UseLoopDelay=false)
   &SpawnBurst_Instantaneous(Spawn Count=2, Use Spawn Probability=false, UseLoopCountLimit=true)
   /SetVariables(Emitter.LineColor:LinearColor=<rgba 9 1.7338 0 1>) {isSetParameters=true}
  ]
  particleSpawnScript [
   &PartitionParticles()
   &InitializeParticle(Color=<rgba 10 10 10 1>, Sprite Size=<10 50>, Sprite Size Mode=::NewEnumerator4, UsePositionOffset=false, Write Material Random=true)
   /SetVariables(Particles.UV Offset:NiagaraFloat=&Lerp_Float(
     A=&RandomRangeFloat(Minimum=0, Maximum=1)
     B=&RandomRangeFloat(Minimum=0, Maximum=1)
     Alpha=&MakeFloatFromInt(Int=$Particles.PartitionParticles.Partition)
    )) {isSetParameters=true, module=SetVariables_#4}
  ]
  particleUpdateScript [
   &ParticleState(
    Kill Particles When Lifetime Has Elapsed=false
    Let Infinitely Lived Particles Die When Emitter Deactivates=true
    Loop Particles Lifetime=true
   )
   /SetVariables(Particles.SpriteAlignment=/Niagara/DynamicInputs/Vectors/NegateVector(Vector:Vector3f=$System.DiveVelocityScale)) {isSetParameters=true, module=SetVariables_#5}
   &SkeletalMeshLocation(
    Bone / Socket Index=&SelectIntFromArray(Direct Array Index=$Particles.PartitionParticles.Partition, Int Selection Array=DataInterface() {propertyValues {IntData [4,5]}})
    Bone Sampling Mode=::NewEnumerator4
    MaskSpawnGroup=false
    Sampled Position Offset=&FN_FrontEndVector3Modifier(
     Gameplay=&Lerp_Vector(A=<0 0 0>, B=<1 1 0>, Alpha=$User.ParaGlideLeanAlpha)
     Front End=&Lerp_Vector(A=<0 0 0>, B=<1 1 1>, Alpha=$User.ParaGlideLeanAlpha)
    )
    Sampled Position Offset Coordinate Space:ENiagaraSkelMeshTransforms=::NewEnumerator1
    Skeletal Mesh=$User.SkeletalMesh
   ) {module=SkeletalMeshLocation001}
   &DynamicMaterialParameters(
    Index 0 Param 1=0
    Index 1 Param 1=1
    Index 1 Param 2=&Lerp_Float(
     A=0
     B=1
     Alpha=&SelectFloatFromArray(
      Array Sampling Mode=::NewEnumerator0
      Direct Array Index=?
      Float Selection Array=DataInterface() {propertyValues {FloatData [-1,1]}}
      Random Seed=0
     )
    )
    Module.Param0WriteEnabled=true
    Module.Param0WriteEnabled1=true
    Module.Param1WriteEnabled=true
    Module.Param1WriteEnabled1=false
    Module.Param2WriteEnabled=true
    Module.Param2WriteEnabled1=false
    Module.Param3WriteEnabled=true
    Module.Param3WriteEnabled1=false
    Write Parameter Index 1=true
   )
  ]
  rendererClasses [&NiagaraSpriteRendererProperties]
  renderers [{rendererClass=&NiagaraSpriteRendererProperties}]
 }
 Trail_Forearm {
  emitterUpdateScript [
   &EmitterState()
   &SpawnRate(Spawn Probability=0.5, SpawnRate=&RandomRangeFloat(Minimum=1, Maximum=2), Use Spawn Probability=true)
  ]
  particleSpawnScript [
   &PartitionParticles()
   &InitializeParticle(
    Color=<rgba 0.0908 0.9911 0.9473 1>
    Lifetime Max=4
    Lifetime Min=2
    Lifetime Mode=::NewEnumerator1
    Sprite Size=&Multiply_Vector2D(
     A=&Lerp_Vector2(A=<7 75>, B=<15 100>, Alpha=&RandomRangeFloat(Minimum=0, Maximum=1))
     B=&Lerp_Vector2(A=<1 1>, B=<-1 1>, Alpha=&MakeFloatFromInt(Int=$Particles.PartitionParticles.Partition))
    )
    Sprite Size Mode=::NewEnumerator4
    UsePositionOffset=false
    Write Material Random=true
   )
   /SetVariables(Particles.UV Offset:NiagaraFloat=&Lerp_Float(
     A=&RandomRangeFloat(Minimum=0, Maximum=1)
     B=&RandomRangeFloat(Minimum=0, Maximum=1)
     Alpha=&MakeFloatFromInt(Int=$Particles.PartitionParticles.Partition)
    )) {isSetParameters=true, module=SetVariables_#4}
  ]
  particleUpdateScript [
   &ParticleState()
   /SetVariables(Particles.SpriteAlignment=&NegateVector(Vector:Vector3f=$System.DiveVelocityScale)) {isSetParameters=true, module=SetVariables_#5}
   &SkeletalMeshLocation(
    Bone / Socket Index=&SelectIntFromArray(Direct Array Index=$Particles.PartitionParticles.Partition, Int Selection Array=DataInterface() {propertyValues {IntData [0,1]}})
    Bone Sampling Mode=::NewEnumerator4
    MaskSpawnGroup=false
    Sampled Position Offset=&Multiply_VectorByFloat(
     Vector=&Lerp_Vector(A=<0 0 0>, B=<22 0 0>, Alpha=&RandomRangeFloat(Minimum=0, Maximum=1))
     Float=&SelectFloatFromArray(Direct Array Index=$Particles.PartitionParticles.Partition, Float Selection Array=DataInterface() {propertyValues {FloatData [1,-1]}})
    )
    Skeletal Mesh=$User.SkeletalMesh
   )
   &ScaleSpriteSize(
    Scale Factor=&Multiply_Vector2D(
     A=&Lerp_Vector2(A=<1 1>, B=<0.75 1.5>, Alpha=0.5)
     B=&MakeVector2D(
      X=&FloatFromCurve(
       FloatCurve=DataInterface() {propertyValues {Curve {keys [{arriveTangent=2.1688, interpMode=RCIM_Cubic, time=0, value=0.33}, <at 1.5 1 RCIM_Cubic>]}}}
       CurveIndex=$Emitter.Age
      )
      Y=&FN_FrontEndFloatModifier(Gameplay=1, Front End=0.75)
     )
    )
    Scale Sprite Size Mode=::NewEnumerator2
   )
   &ScaleColor(
    Scale Alpha=&FloatFromCurve(
     FloatCurve=DataInterface() {propertyValues {Curve {keys [{interpMode=RCIM_Cubic, tangentMode=RCTM_User, time=0, value=0}, <at 0.1 1>, {interpMode=RCIM_Cubic, tangentMode=RCTM_User, time=1, value=0}]}}}
     Scale Curve=3
    )
    ScaleA=true
    ScaleRGB=false
   ) {module=ScaleColor003}
   &DynamicMaterialParameters(
    Index 1 Param 2=1
    Module.Param0WriteEnabled1=true
    Module.Param1WriteEnabled1=true
    Module.Param2WriteEnabled1=false
    Module.Param3WriteEnabled1=false
    Write Parameter Index 0=false
    Write Parameter Index 1=true
   ) {module=DynamicMaterialParameters002}
  ]
  rendererClasses [&NiagaraSpriteRendererProperties]
  renderers [{rendererClass=&NiagaraSpriteRendererProperties}]
 }
 Trail_Upperarm {
  emitterUpdateScript [
   &EmitterState(Life Cycle Mode=::NewEnumerator1, Loop Behavior=::NewEnumerator1, Loop Duration Mode=::NewEnumerator1, UseLoopDelay=false)
   &SpawnRate(Spawn Probability=0.75, SpawnRate=2, Use Spawn Probability=true)
  ]
  particleSpawnScript [
   &PartitionParticles()
   &InitializeParticle(
    Color=<rgba 10 10 10 1>
    Lifetime Max=4
    Lifetime Min=2
    Lifetime Mode=::NewEnumerator1
    Sprite Size=&Multiply_Vector2D(
     A=&Lerp_Vector2(A=<17 100>, B=<25 190>, Alpha=&RandomRangeFloat(Minimum=0, Maximum=1))
     B=&Lerp_Vector2(A=<1 1>, B=<-1 1>, Alpha=&MakeFloatFromInt(Int=$Particles.PartitionParticles.Partition))
    )
    Sprite Size Mode=::NewEnumerator4
    UsePositionOffset=false
    Write Material Random=true
   )
   /SetVariables(Particles.UV Offset:NiagaraFloat=&Lerp_Float(
     A=&RandomRangeFloat(Minimum=0, Maximum=1)
     B=&RandomRangeFloat(Minimum=0, Maximum=1)
     Alpha=&MakeFloatFromInt(Int=$Particles.PartitionParticles.Partition)
    )) {isSetParameters=true, module=SetVariables_#4}
  ]
  particleUpdateScript [
   &ParticleState()
   /SetVariables(Particles.SpriteAlignment=&NegateVector(Vector:Vector3f=$System.DiveVelocityScale)) {isSetParameters=true, module=SetVariables_#5}
   &SkeletalMeshLocation(
    Bone / Socket Index=&SelectIntFromArray(Direct Array Index=$Particles.PartitionParticles.Partition, Int Selection Array=DataInterface() {propertyValues {IntData [2,3]}})
    Bone Sampling Mode=::NewEnumerator4
    MaskSpawnGroup=false
    Sampled Position Offset=&Multiply_VectorByFloat(
     Vector=&Lerp_Vector(A=<0 0 0>, B=<26 0 0>, Alpha=&RandomRangeFloat(Minimum=0, Maximum=1))
     Float=&SelectFloatFromArray(Direct Array Index=$Particles.PartitionParticles.Partition, Float Selection Array=DataInterface() {propertyValues {FloatData [1,-1]}})
    )
    Skeletal Mesh=$User.SkeletalMesh
   )
   &ScaleSpriteSize(
    Scale Factor=&Multiply_Vector2D(
     A=&Lerp_Vector2(A=<1 1>, B=<0.75 1.5>, Alpha=0.5)
     B=&MakeVector2D(
      X=&FloatFromCurve(
       FloatCurve=DataInterface() {propertyValues {Curve {keys [{arriveTangent=2.1688, interpMode=RCIM_Cubic, time=0, value=0.33}, <at 1.5 1 RCIM_Cubic>]}}}
       CurveIndex=$Emitter.Age
      )
      Y=&FN_FrontEndFloatModifier(Gameplay=1, Front End=0.75)
     )
    )
    Scale Sprite Size Mode=::NewEnumerator2
   )
   &ScaleColor(
    Scale Alpha=&FloatFromCurve(
     FloatCurve=DataInterface() {propertyValues {Curve {keys [{interpMode=RCIM_Cubic, tangentMode=RCTM_User, time=0, value=0}, <at 0.1 1>, {interpMode=RCIM_Cubic, tangentMode=RCTM_User, time=1, value=0}]}}}
     Scale Curve=3
    )
    ScaleA=true
    ScaleRGB=false
   ) {module=ScaleColor004}
   &DynamicMaterialParameters(
    Index 1 Param 1=&Lerp_Float(
     A=0
     B=0
     Alpha=&SelectFloatFromArray(
      Array Sampling Mode=::NewEnumerator0
      Direct Array Index=$Particles.PartitionParticles.Partition
      Float Selection Array=DataInterface() {propertyValues {FloatData [-1,1]}}
      Random Seed=0
     )
    )
    Index 1 Param 2=&Lerp_Float(
     A=0
     B=1
     Alpha=&SelectFloatFromArray(
      Array Sampling Mode=::NewEnumerator0
      Direct Array Index=?
      Float Selection Array=DataInterface() {propertyValues {FloatData [-1,1]}}
      Random Seed=0
     )
    )
    Module.Param0WriteEnabled1=true
    Module.Param1WriteEnabled1=true
    Module.Param2WriteEnabled1=false
    Module.Param3WriteEnabled1=false
    Write Parameter Index 0=false
    Write Parameter Index 1=true
   )
  ]
  rendererClasses [&NiagaraSpriteRendererProperties]
  renderers [{rendererClass=&NiagaraSpriteRendererProperties}]
 }
}
summary {
 systemName=NS_Contrail_VitalPsych
 emitters [
  {emitterName=SoftGlow, rendererClasses [&NiagaraSpriteRendererProperties]}
  {emitterName=Trail_Upperarm, rendererClasses [&NiagaraSpriteRendererProperties]}
  {emitterName=Trail_Forearm, rendererClasses [&NiagaraSpriteRendererProperties]}
  {emitterName=Trail_Feet, rendererClasses [&NiagaraSpriteRendererProperties]}
  {emitterName=HandGlow_Decal, rendererClasses [&NiagaraSpriteRendererProperties]}
  {emitterName=Eye_Logo, rendererClasses [&NiagaraSpriteRendererProperties]}
  {emitterName=Light, rendererClasses [&NiagaraLightRendererProperties]}
  {emitterName=GlowRings, rendererClasses [&NiagaraMeshRendererProperties]}
 ]
 userVariables [
  User.bIsFrontEnd {type=NiagaraBool}
  User.bIsFrontEndPreview {type=NiagaraBool, defaultValue=true}
  User.ParaGlideLeanAlpha {type=NiagaraFloat, defaultValue=0, description=""}
  User.PlayerSpeed {type=NiagaraFloat, defaultValue=0, description=""}
  User.RibbonIDOffset {type=NiagaraInt32}
  User.SkeletalMesh {
   type=NiagaraDataInterfaceSkeletalMesh
   defaultValue {
    struct=NiagaraExt_VariableValue_DataInterface
    value {dataInterface=&self:NiagaraDataInterfaceSkeletalMesh_0}
   }
  }
 ]
}
